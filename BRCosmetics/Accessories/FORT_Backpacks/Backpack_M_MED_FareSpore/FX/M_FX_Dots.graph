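# Translucent unlit Niagara mesh-particle material for the FareSpore backpack, drawing radial dots that appear built to react to music.
# A spectrum dynamic parameter (X channel only), shaped by a HeavyRoar curve atlas, combines with a time-driven sine and an Open/Close parameter into a wave mask.
# Named reroutes, apparently waves and contrasted Perlin noise, set the dot texture's smoothstep threshold, multiplied by particle alpha and a V-edge fade.
# Particle color feeds a shared emissive adjustment function.
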Material /BRCosmetics/Accessories/FORT_Backpacks/Backpack_M_MED_FareSpore/FX/M_FX_Dots
graph {
 properties {
  bAutomaticallySetUsageInEditor=false
  bUsedWithNiagaraMeshParticles=true
  blendMode=BLEND_Translucent
  shadingModel=MSM_Unlit
  twoSided=true
  usedShadingModels=Unlit
 }
 nodes [
  n1 = Multiply(A=NamedRerouteUsage(), B=ScalarParameter() {defaultValue=1, parameterName=MainScaler})
  n2 = NamedRerouteUsage()
  Add(A=@n1, B=@n2)
  Constant() {r=0.8}
  n3 = ParticleColor()
  MaterialFunctionCall(Emissive=@n3.RGB) {materialFunction=&MF_FX_EmissiveAdjustments}
  n4 = OneMinus(None=Multiply(A=@n1, B=@n2))
  n5 = Add(A=@n4, B=Constant() {r=0.2})
  n6 = TextureSample() {constCoordinate=1, texture=&/T_FX_Radial_Dots}
  n7 = OneMinus(None=MaterialFunctionCall(Float2=TextureCoordinate()) {materialFunction=&BreakOutFloat2Components}.G)
  n8 = Constant() {r=5}
  Multiply(A=Multiply(A=SmoothStep(Min=@n4, Max=@n5, Value=@n6.R), B=@n3.A), B=Saturate(None=Multiply(A=@n7, B=@n8)))
  n9 = DynamicParameter() {parameterIndex=1, paramNames [Open/Close, Wave Offset, "", ""]}
  Multiply(B=@n9.Open/Close)
  n10 = DynamicParameter() {paramNames [SpectrumSample_x, SpectrumSample_y, SpectrumSample_z, SpectrumSample_a]}
  n11 = ScalarParameter() {defaultValue=2, parameterName=Equaliser Noise Tile}
  n12 = MaterialFunctionCall(Float2=Multiply(
    A=TextureCoordinate()
    B=MaterialFunctionCall(X=@n11) {materialFunction=&MakeFloat2}.Result
   )) {materialFunction=&BreakOutFloat2Components}
  n13 = ScalarParameter() {defaultValue=0.3, parameterName=Equaliser Noise Offset}
  n14 = CurveAtlasRowParameter(CurveTime=Frac(None=Add(A=@n12.R, B=@n13))) {
   atlas=/BRCosmetics/Accessories/FORT_Backpacks/Backpack_M_MED_HeavyRoar/FX/CA_Backpack_HeavyRoar
   curve=/BRCosmetics/Accessories/FORT_Backpacks/Backpack_M_MED_HeavyRoar/FX/CC_Backpack_HeavyRoar_MusicDistort
   defaultValue=1
   parameterName=ColorDistortion
  }
  n15 = MaterialFunctionCall(Float4=Multiply(
    A=MaterialFunctionCall(X=@n10.SpectrumSample_x, Y=@n10.SpectrumSample_x, Z=@n10.SpectrumSample_x, A=@n10.SpectrumSample_x) {materialFunction=&MakeFloat4}.Result
    B=MaterialFunctionCall(X=@n14.R, Y=@n14.R, Z=@n14.R, A=@n14.R) {materialFunction=&MakeFloat4}.Result
   )) {materialFunction=&BreakOutFloat4Components}
  n16 = Max(A=Max(A=@n15.R, B=@n15.R), B=@n15.B)
  MaterialExpressionRemap_0 = MaterialXRemap!(
   None=NamedRerouteDeclaration(None=Max(A=@n16, B=@n15.A)) {name=SpectrumSample, nodeColor=<rgba 0.8281 0 1 1>}
   InputLow=ScalarParameter() {parameterName=InLow}
   InputHigh=ScalarParameter() {defaultValue=1, parameterName=InHigh}
   TargetLow=Constant()
   TargetHigh=Constant() {r=1}
  )
  n17 = Multiply(
   A=MaterialFunctionCall(Float2=TextureCoordinate()) {materialFunction=&BreakOutFloat2Components}.R
   B=ScalarParameter() {defaultValue=-3, parameterName=Sin_01_Freq}
  )
  n18 = Constant() {r=2}
  n19 = Multiply(
   A=Time()
   B=Divide(
    A=MaterialFunctionCall(Multiplier=@n18) {materialFunction=&Pi}.Result
    B=ScalarParameter() {defaultValue=4, parameterName=Sin_01_Period}
   )
  )
  n20 = Multiply(A=Sine(None=Add(A=@n17, B=@n19)), B=ScalarParameter() {defaultValue=1, parameterName=Sin_01_Scale})
  n21 = ScalarParameter() {parameterName=Sin_01_Bias}
  n22 = Power(Base=ConstantBiasScale(None=Add(A=@n20, B=@n21)), Exp=Constant() {r=1})
  n23 = ScalarParameter() {defaultValue=-0.4, parameterName=Wave Distortion Amount}
  n24 = Saturate(None=Add(A=@MaterialExpressionRemap_0, B=Multiply(A=@n22, B=@n23) {constB=0.1}) {constA=1, constB=0})
  n25 = Multiply(A=ScalarParameter() {defaultValue=1, parameterName=WavesAmount}, B=@n9.Open/Close)
  n26 = Multiply(
   A=OneMinus(None=Multiply(A=@n24, B=@n25))
   B=LinearInterpolate(
    B=ScalarParameter() {defaultValue=0.5, parameterName=Param_1}
    Alpha=DynamicParameter() {parameterIndex=1, paramNames [Open/Close, Wave Offset, "", ""]}.Open/Close
   ) {constA=1}
  )
  n27 = SmoothStep(
   Min=@n26
   Max=Add(A=@n26, B=ScalarParameter() {defaultValue=1, parameterName=Wave_01 Smothness})
   Value=MaterialFunctionCall(Float2=TextureCoordinate()) {materialFunction=&BreakOutFloat2Components}.G
  )
  NamedRerouteDeclaration(None=@n27) {name=Waves, nodeColor=<rgba 0.2188 0 1 1>}
  NamedRerouteDeclaration(None=@n27) {name=TEST, nodeColor=<rgba 1 0.6094 0 1>}
  n28 = TextureObjectParameter() {
   group=Hue
   parameterName=Hue_Texture
   samplerType=SAMPLERTYPE_LinearGrayscale
   sortPriority=0
   texture=&T_SimplePerlin_Greyscale
  }
  n29 = VectorParameter() {parameterName=Noise_UVControl, defaultValue=<rgba 0.5 0.5 0.5 0>}
  n30 = MaterialFunctionCall(
   In=MaterialFunctionCall(Texture=@n28, UVControl_RGB=@n29.RGB, UVControl_A=@n29.RGB) {materialFunction=&FXCommon_TextureSample}.R
   Contrast=Constant() {r=2}
  ) {materialFunction=&CheapContrast}
  NamedRerouteDeclaration(None=LinearInterpolate(
    A=ScalarParameter() {defaultValue=0.8, parameterName=NoiseMin}
    B=ScalarParameter() {defaultValue=1, parameterName=NoiseMax}
    Alpha=Saturate(None=@n30.Result)
   )) {name=Noise, nodeColor=<rgba 1 0.6797 0 1>}
 ]
}
