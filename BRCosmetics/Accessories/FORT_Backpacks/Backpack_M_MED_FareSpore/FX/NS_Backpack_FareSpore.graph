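# Audio-reactive Niagara effect for the FareSpore backpack cosmetic, with a music equalizer, glows and sound waves.
# Mirrored Base, Dots and Equalizer mesh pairs receive System.SpectrumMusic and smoothed reactivity through dynamic material parameters; two Glow sprites pulse with reactivity.
# Arms_SoundWaves and PostLine002 burst purple and cyan sprites, gated by a reactivity threshold, passing world-space endpoints to materials.
# A system Timeline scales opacity, and a fake-audio loop module is included.
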NiagaraSystem /BRCosmetics/Accessories/FORT_Backpacks/Backpack_M_MED_FareSpore/FX/NS_Backpack_FareSpore
dependencies {
 usedDataInterfaces [&NiagaraDataInterfaceCurve, &NiagaraDataInterfaceArrayInt32, &NiagaraDataInterfaceColorCurve]
 usedDynamicInputs [
  /Niagara/DynamicInputs/Lerp/SmoothLerpOverTime_Float
  &MakeFloatFromLinearColor
  &RemapRange
  &Lerp_Float
  &ColorFromCurve
  &MakeCustomLinearColorFromBool
  &LerpLinearColors
 ]
 usedModules [
  &self:SystemSpawnScript.SystemScriptSource.SystemScriptGraph.NiagaraNodeAssignment_3.SetVariables_#1
  &self:SystemSpawnScript.SystemScriptSource.SystemScriptGraph.NiagaraNodeAssignment_2.SetVariables_#2
  &SystemState
  &Timeline
  /Game/Effects/Niagara/Modules/Audio/FN_FakeAudioReactivityLoop
  &self:SystemSpawnScript.SystemScriptSource.SystemScriptGraph.NiagaraNodeAssignment_8.SetVariables_#3
  &self:SystemSpawnScript.SystemScriptSource.SystemScriptGraph.NiagaraNodeAssignment_0.SetVariables_#4
  &self:SystemSpawnScript.SystemScriptSource.SystemScriptGraph.NiagaraNodeAssignment_1.SetVariables_#5
  &self:SystemSpawnScript.SystemScriptSource.SystemScriptGraph.NiagaraNodeAssignment_6.SetVariables_#6
  &self:SystemSpawnScript.SystemScriptSource.SystemScriptGraph.NiagaraNodeAssignment_7.SetVariables_#7
  &InterpolateOverTime
 ]
}
emitters {
 Arms_SoundWaves {
  emitterUpdateScript [
   &EmitterState(Inactive Response=::NewEnumerator1, Life Cycle Mode=::NewEnumerator1, Loop Duration=0.5, UseLoopDelay=false)
   &SpawnBurst_Instantaneous(
    Spawn Count=2
    Spawn Probability=&MakeCustomFloatFromBool(Bool=&SetBoolByFloatComparison(A=$System.AudioReactivity, B=0.1), True Float=1, False Float=0)
    Use Spawn Probability=true
    UseLoopCountLimit=false
   )
   &SpawnBurst_Instantaneous(
    Spawn Count=2
    Spawn Probability=&MakeCustomFloatFromBool(Bool=&SetBoolByFloatComparison(A=$System.AudioReactivity, B=0.1), True Float=1, False Float=0)
    Spawn Time=0.2
    Use Spawn Probability=true
    UseLoopCountLimit=false
   ) {module=SpawnBurst_Instantaneous001}
  ]
  particleSpawnScript [
   &PartitionParticles(Particles Per Partition=1, Partition Mode:ENiagara_PartitionMode_v2=::NewEnumerator1)
   &InitializeParticle(
    Color=&MergeHelper_ColorSwitch(0=<rgba 0.6667 0 1 1>, 1=<rgba 0 0.6667 1 1>, 2=?, 3=?, Index=$Particles.PartitionParticles.Partition, Num=::NewEnumerator0)
    Lifetime=0.5
    Position Offset=&MergeHelper_VectorSwitch(0=<-7 -17 -10>, 1=<7 -17 -10>, 2=?, 3=?, Index=$Particles.PartitionParticles.Partition, Num=::NewEnumerator0)
    Position Offset Coordinate Space=::Local
    Sprite Size Mode=::NewEnumerator3
    Uniform Sprite Size=&MergeHelper_FloatSwitch(0=30, 1=30, 2=?, 3=?, Index=$Particles.PartitionParticles.Partition, Num=::NewEnumerator0)
    UsePositionOffset=true
    Write Material Random=true
   )
   /SetVariables(Particles.VisibilityTag=&MakeCustomIntFromBool(Bool=$User.bIsFrontEndPreview, True INT=1, False INT=0)) {isSetParameters=true}
  ]
  particleUpdateScript [
   &ParticleState()
   &SkeletalMeshLocation(
    Bone / Socket Index=&SelectIntFromArray(
     Direct Array Index=$Particles.PartitionParticles.Partition
     Direct Array Mode=::NewEnumerator1
     Int Selection Array=DataInterface() {propertyValues {IntData [8,9,10,11]}}
    )
    MaskSpawnGroup=false
    Mesh Sampling Type=::NewEnumerator5
    Sampled Position Offset=<0 0 0>
    Skeletal Mesh=$System.Skeletal Mesh
    Socket Sampling Mode=::NewEnumerator5
   ) {enabled=false}
   &CameraOffset(Camera Offset Amount=&FN_FrontEndFloatModifier(Gameplay=20, Front End=0))
   /SetVariables(Particles.OffsetPosition:NiagaraPosition=&AddVectorToPosition(
     Position=$Particles.Position
     Vector=&MergeHelper_VectorSwitch(0=<5 0 3>, 1=<-5 0 3>, 2=?, 3=?, Index=$Particles.PartitionParticles.Partition, Num=::NewEnumerator0)
    )) {isSetParameters=true, module=SetVariables_#8}
   /SetVariables(
    Particles.WSOffsetPosition:NiagaraPosition=&TransformPosition(Position=$StackContext.OffsetPosition, Source Space=::Local, Destination Space=::World)
    Particles.WSPosition:NiagaraPosition=&TransformPosition(Position=$StackContext.Position, Source Space=::Local, Destination Space=::World)
   ) {isSetParameters=true}
   &DynamicMaterialParameters(
    Index 0 Param 1=&MakeFloatFromVector(Vector=Linked() {linkedVariable {name=StackContext.WSPosition, type=NiagaraPosition}}, Channel=::NewEnumerator0)
    Index 0 Param 2=&MakeFloatFromVector(Vector=Linked() {linkedVariable {name=StackContext.WSPosition, type=NiagaraPosition}}, Channel=::NewEnumerator1)
    Index 0 Param 3=&MakeFloatFromVector(Vector=Linked() {linkedVariable {name=StackContext.WSPosition, type=NiagaraPosition}}, Channel=::NewEnumerator2)
    Index 0 Param 4=&RandomRangeFloat(Minimum=0, Maximum=1)
    Index 1 Param 1=&MakeFloatFromVector(Vector=Linked() {linkedVariable {name=StackContext.WSOffsetPosition, type=NiagaraPosition}}, Channel=::NewEnumerator0)
    Index 1 Param 2=&MakeFloatFromVector(Vector=Linked() {linkedVariable {name=StackContext.WSOffsetPosition, type=NiagaraPosition}}, Channel=::NewEnumerator1)
    Index 1 Param 3=&MakeFloatFromVector(Vector=Linked() {linkedVariable {name=StackContext.WSOffsetPosition, type=NiagaraPosition}}, Channel=::NewEnumerator2)
    Index 1 Param 4=&FloatFromCurve(FloatCurve=DataInterface() {propertyValues {Curve {keys [
         {arriveTangent=-0.8006, interpMode=RCIM_Cubic, leaveTangent=-0.8006, tangentMode=RCTM_User, time=0, value=1}
         {arriveTangent=-0.3233, interpMode=RCIM_Cubic, leaveTangent=-0.3233, tangentMode=RCTM_User, time=1, value=0.5}
        ]}}})
    Module.Index0Float0Write=true
    Module.Index0Float1Write=true
    Module.Index0Float2Write=true
    Module.Index0Float3Write=true
    Module.Index1Float0Write=true
    Module.Index1Float1Write=true
    Module.Index1Float2Write=true
    Module.Index1Float3Write=true
    Module.Index2Float0Write=true
    Module.Index2Float1Write=false
    Module.Index2Float2Write=false
    Module.Index2Float3Write=false
    Write Parameter Index 1=::NewEnumerator1
    Write Parameter Index 2=::NewEnumerator1
   )
   &ScaleColor(
    Scale Alpha=&FloatFromCurve(
     FloatCurve=DataInterface() {propertyValues {Curve {keys [
         {arriveTangent=0.5632, interpMode=RCIM_Cubic, leaveTangent=0.5632, tangentMode=RCTM_User, time=0, value=0}
         {arriveTangent=0.168, interpMode=RCIM_Cubic, leaveTangent=0.168, tangentMode=RCTM_User, time=0.25, value=1}
         {arriveTangent=-0.1012, interpMode=RCIM_Cubic, leaveTangent=-0.1012, tangentMode=RCTM_User, time=1, value=0}
        ]}}}
     Scale Curve=&Multiply_Float(A=&FN_FrontEndFloatModifier(Gameplay=5, Front End=5), B=$System.Timeline001.Time)
    )
    Scale RGB=&VectorFromFloat(Value=&FloatFromCurve(FloatCurve=DataInterface() {propertyValues {Curve {keys [
          {arriveTangent=0.168, interpMode=RCIM_Cubic, leaveTangent=0.168, tangentMode=RCTM_User, time=0.2, value=1.5}
          {arriveTangent=-0.1012, interpMode=RCIM_Cubic, leaveTangent=-0.1012, tangentMode=RCTM_User, time=0.4, value=1}
         ]}}}))
    ScaleA=true
    ScaleRGB=true
   )
   &ScaleColor(
    Scale Alpha=$System.AudioReactivity
    Scale RGB=&VectorFromFloat(Value=&Add_Float(
      A=&FloatFromCurve(
       FloatCurve=DataInterface() {propertyValues {Curve {keys [
           {interpMode=RCIM_Cubic, tangentMode=RCTM_Break, time=0, value=0}
           {arriveTangent=3.7964, interpMode=RCIM_Cubic, leaveTangent=1.1533, tangentMode=RCTM_Break, time=0.9, value=1}
           <at 1 0>
          ]}}}
       CurveIndex=$System.AudioReactivity
      )
      B=2
     ))
    ScaleA=true
    ScaleRGB=true
   ) {module=ScaleColor001}
  ]
  rendererClasses [&NiagaraSpriteRendererProperties]
  renderers [
   {rendererClass=&NiagaraSpriteRendererProperties}
   {rendererClass=&NiagaraSpriteRendererProperties, rendererIndex=1}
   {rendererClass=&NiagaraSpriteRendererProperties, rendererIndex=2}
  ]
 }
 Base {
  emitterUpdateScript [
   &EmitterState(
    Inactive Response=::NewEnumerator1
    Life Cycle Mode=::NewEnumerator1
    Loop Behavior=::NewEnumerator1
    Loop Duration Mode=::NewEnumerator1
    Scalability Mode=::NewEnumerator1
    UseLoopDelay=false
   )
   &SpawnBurst_Instantaneous(Spawn Count=2, Use Spawn Probability=false, UseLoopCountLimit=false)
  ]
  particleSpawnScript [
   &FN_SetSpawnGroup(Number of Spawn Groups=2)
   &InitializeParticle(
    Color Mode=::NewEnumerator0
    Lifetime=5
    Mesh Scale=&MergeHelper_VectorSwitch(0=<1 1 1>, 1=<-1 1 1>, 2=?, 3=?, Index=$Particles.SpawnGroup, Num=::NewEnumerator0)
    Mesh Scale Mode=::NewEnumerator4
    UsePositionOffset=false
    Write Material Random=true
   )
  ]
  particleUpdateScript [
   &ParticleState(Kill Particles When Lifetime Has Elapsed=false)
   &CameraOffset()
   &DynamicMaterialParameters(
    Index 0 Param 1=&MakeFloatFromVector4(Vector 4=&MakeVector4_FromLinearColor(LinearColor=$System.SpectrumMusic), Channel=::NewEnumerator0)
    Index 0 Param 2=&MakeFloatFromVector4(Vector 4=&MakeVector4_FromLinearColor(LinearColor=$System.SpectrumMusic), Channel=::NewEnumerator1)
    Index 0 Param 3=&MakeFloatFromVector4(Vector 4=&MakeVector4_FromLinearColor(LinearColor=$System.SpectrumMusic), Channel=::NewEnumerator2)
    Index 0 Param 4=&MakeFloatFromVector4(Vector 4=&MakeVector4_FromLinearColor(LinearColor=$System.SpectrumMusic), Channel=::NewEnumerator3)
    Index 1 Param 1=&MakeCustomFloatFromBool(Bool=$User.DebugReactivity, True Float=1, False Float=$System.InterpolateOverTime.InterpolateOverTime_Float.Moving Average)
    Module.Index0Float0Write=true
    Module.Index0Float1Write=true
    Module.Index0Float2Write=true
    Module.Index0Float3Write=true
    Module.Index1Float0Write=true
    Module.Index1Float1Write=true
    Module.Index1Float2Write=true
    Module.Index1Float3Write=true
    Write Parameter Index 1=::NewEnumerator1
   )
   /SetVariables(
    Particles.BaseColor:LinearColor=&MultiplyLinearColor(
     Linear Color=<rgba 0.6 0.6 0.6 0.2>
     Scale Factor=&MakeVector4FromVectorAndFloat("Vector (RGB)"=<1 1 1>, "Float (Alpha)"=$System.Timeline001.Time)
    )
    Particles.GlowColor=<rgba 1.1 1.1 1.1 1>
   ) {isSetParameters=true, module=SetVariables_#9}
  ]
  rendererClasses [&NiagaraMeshRendererProperties]
  renderers [{rendererClass=&NiagaraMeshRendererProperties, rendererIndex=0}]
 }
 Dots {
  emitterUpdateScript [
   &EmitterState(
    Inactive Response=::NewEnumerator1
    Life Cycle Mode=::NewEnumerator1
    Loop Behavior=::NewEnumerator1
    Loop Duration Mode=::NewEnumerator1
    Scalability Mode=::NewEnumerator1
    UseLoopDelay=false
   )
   &SpawnBurst_Instantaneous(Spawn Count=2, Use Spawn Probability=false, UseLoopCountLimit=false)
  ]
  particleSpawnScript [
   &FN_SetSpawnGroup(Number of Spawn Groups=2)
   &InitializeParticle(
    Color=<rgba 0.8203 0.4609 1 0.3>
    Lifetime=5
    Mesh Scale=&MergeHelper_VectorSwitch(0=<1 1 1>, 1=<-1 1 1>, 2=?, 3=?, Index=$Particles.SpawnGroup, Num=::NewEnumerator0)
    Mesh Scale Mode=::NewEnumerator4
    UsePositionOffset=false
    Write Material Random=true
   )
  ]
  particleUpdateScript [
   &ParticleState(Kill Particles When Lifetime Has Elapsed=false)
   &CameraOffset(Camera Offset Amount=-1)
   &DynamicMaterialParameters(
    Index 0 Param 1=&MakeFloatFromVector4(Vector 4=&MakeVector4_FromLinearColor(LinearColor=$System.SpectrumMusic), Channel=::NewEnumerator0)
    Index 0 Param 2=&MakeFloatFromVector4(Vector 4=&MakeVector4_FromLinearColor(LinearColor=$System.SpectrumMusic), Channel=::NewEnumerator1)
    Index 0 Param 3=&MakeFloatFromVector4(Vector 4=&MakeVector4_FromLinearColor(LinearColor=$System.SpectrumMusic), Channel=::NewEnumerator2)
    Index 0 Param 4=&MakeFloatFromVector4(Vector 4=&MakeVector4_FromLinearColor(LinearColor=$System.SpectrumMusic), Channel=::NewEnumerator3)
    Index 1 Param 1=&MakeCustomFloatFromBool(Bool=$User.DebugReactivity, True Float=1, False Float=$System.InterpolateOverTime.InterpolateOverTime_Float.Moving Average)
    Module.Index0Float0Write=true
    Module.Index0Float1Write=true
    Module.Index0Float2Write=true
    Module.Index0Float3Write=true
    Module.Index1Float0Write=true
    Module.Index1Float1Write=true
    Module.Index1Float2Write=true
    Module.Index1Float3Write=true
    Write Parameter Index 1=::NewEnumerator1
   )
   &ScaleColor(
    Scale Alpha=&Multiply_Float(
     A=&FNTimeOfDayScalar(Night Scale=1, Morning Scale=1, Evening Scale=1)
     B=$System.Timeline001.Time
    )
    Scale RGB=&VectorFromFloat(Value=&Add_Float(
      A=&FloatFromCurve(
       FloatCurve=DataInterface() {propertyValues {Curve {keys [
           {interpMode=RCIM_Cubic, tangentMode=RCTM_Break, time=0, value=0}
           {arriveTangent=3.7964, interpMode=RCIM_Cubic, leaveTangent=1.1533, tangentMode=RCTM_Break, time=0.9, value=1}
           <at 1 0>
          ]}}}
       CurveIndex=$System.AudioReactivity
      )
      B=1
     ))
    ScaleA=true
    ScaleRGB=true
   ) {module=ScaleColor001}
  ]
  rendererClasses [&NiagaraMeshRendererProperties]
  renderers [{rendererClass=&NiagaraMeshRendererProperties, rendererIndex=0}]
 }
 Equalizer {
  emitterUpdateScript [
   &EmitterState(
    Inactive Response=::NewEnumerator1
    Life Cycle Mode=::NewEnumerator1
    Loop Behavior=::NewEnumerator1
    Loop Duration Mode=::NewEnumerator1
    Scalability Mode=::NewEnumerator1
    UseLoopDelay=false
   )
   &SpawnBurst_Instantaneous(Spawn Count=2, Use Spawn Probability=false, UseLoopCountLimit=false)
  ]
  particleSpawnScript [
   &FN_SetSpawnGroup(Number of Spawn Groups=2)
   &InitializeParticle(
    Color=<rgba 1 1 1 0.13>
    Lifetime=5
    Mesh Scale=&MergeHelper_VectorSwitch(0=<1 1 1>, 1=<-1 1 1>, 2=?, 3=?, Index=$Particles.SpawnGroup, Num=::NewEnumerator0)
    Mesh Scale Mode=::NewEnumerator4
    UsePositionOffset=false
    Write Material Random=true
   )
  ]
  particleUpdateScript [
   &ParticleState(Kill Particles When Lifetime Has Elapsed=false)
   &CameraOffset()
   &DynamicMaterialParameters(
    Index 0 Param 1=&MakeFloatFromVector4(Vector 4=&MakeVector4_FromLinearColor(LinearColor=$System.SpectrumMusic), Channel=::NewEnumerator0)
    Index 0 Param 2=&MakeFloatFromVector4(Vector 4=&MakeVector4_FromLinearColor(LinearColor=$System.SpectrumMusic), Channel=::NewEnumerator1)
    Index 0 Param 3=&MakeFloatFromVector4(Vector 4=&MakeVector4_FromLinearColor(LinearColor=$System.SpectrumMusic), Channel=::NewEnumerator2)
    Index 0 Param 4=&MakeFloatFromVector4(Vector 4=&MakeVector4_FromLinearColor(LinearColor=$System.SpectrumMusic), Channel=::NewEnumerator3)
    Index 1 Param 1=&MakeCustomFloatFromBool(Bool=$User.DebugReactivity, True Float=1, False Float=$System.InterpolateOverTime.InterpolateOverTime_Float.Moving Average)
    Module.Index0Float0Write=true
    Module.Index0Float1Write=true
    Module.Index0Float2Write=true
    Module.Index0Float3Write=true
    Module.Index1Float0Write=true
    Module.Index1Float1Write=true
    Module.Index1Float2Write=true
    Module.Index1Float3Write=true
    Write Parameter Index 1=::NewEnumerator1
   )
   /SetVariables(
    Particles.BaseColor:LinearColor=&MultiplyLinearColor(
     Linear Color=<rgba 0.5 0.5 0.5 1>
     Scale Factor=&MakeVector4FromVectorAndFloat("Vector (RGB)"=<1 1 1>, "Float (Alpha)"=$System.Timeline001.Time)
    )
    Particles.GlowColor=&Multiply_LinearColorByFloat(
     Linear Color=&MultiplyLinearColor(
      Linear Color=<rgba 1.1 1.1 1.1 1>
      Scale Factor=&MakeVector4FromVectorAndFloat(
       "Vector (RGB)"=<1 1 1>
       "Float (Alpha)"=&Multiply_Float(A=$System.Timeline001.Time, B=&FN_FrontEndFloatModifier(Gameplay=1, Front End=0.5))
      )
     )
     Float=&Multiply_Float(A=&FNTimeOfDayScalar(Night Scale=0.7, Morning Scale=0.7, Evening Scale=0.7), B=1)
    )
   ) {isSetParameters=true, module=SetVariables_#9}
   &ScaleColor(
    Scale Alpha=&Multiply_Float(
     A=&FNTimeOfDayScalar(Night Scale=1, Morning Scale=1, Evening Scale=1)
     B=$System.Timeline001.Time
    )
    Scale RGB=&VectorFromFloat(Value=&Add_Float(
      A=&FloatFromCurve(
       FloatCurve=DataInterface() {propertyValues {Curve {keys [
           {interpMode=RCIM_Cubic, tangentMode=RCTM_Break, time=0, value=0}
           {arriveTangent=3.7964, interpMode=RCIM_Cubic, leaveTangent=1.1533, tangentMode=RCTM_Break, time=0.9, value=1}
           <at 1 0>
          ]}}}
       CurveIndex=$System.AudioReactivity
      )
      B=0.5
     ))
    ScaleA=true
    ScaleRGB=true
   ) {module=ScaleColor001}
  ]
  rendererClasses [&NiagaraMeshRendererProperties]
  renderers [
   {rendererClass=&NiagaraMeshRendererProperties, rendererIndex=0}
   {rendererClass=&NiagaraMeshRendererProperties, rendererIndex=1}
   {rendererClass=&NiagaraMeshRendererProperties, rendererIndex=2}
  ]
 }
 FN_Debug {
  bEnabled=false
  emitterUpdateScript [
   &EmitterState()
   &SpawnBurst_Instantaneous(Spawn Count=1, Use Spawn Probability=false)
  ]
  particleSpawnScript [/Niagara/Modules/Spawn/Initialization/InitializeParticle(
    Lifetime=5
    Module.Write Color=true
    Module.Write Lifetime=true
    Module.Write Mass=true
    Module.Write Material Random=true
    Module.Write Position=false
    Module.Write Scale=false
    Module.Write SpriteRotation=false
    Module.Write SpriteSize=true
    Position=&SimulationPosition()
    Sprite Size=<100 100>
   )]
  particleUpdateScript [
   &ParticleState()
   /SetVariables(Particles.Debug:NiagaraFloat=$System.Timeline001.Time) {isSetParameters=true}
   &DynamicMaterialParameters(
    Index 0 Param 1=$Particles.Debug
    Module.Param0WriteEnabled=true
    Module.Param1WriteEnabled=true
    Module.Param2WriteEnabled=true
    Module.Param3WriteEnabled=true
   )
   &SystemLocation(MaskSpawnGroup=false, Offset=<50 0 -100>)
  ]
  rendererClasses [&NiagaraSpriteRendererProperties]
  renderers [{rendererClass=&NiagaraSpriteRendererProperties}]
 }
 Glow {
  emitterUpdateScript [
   &EmitterState(
    Inactive Response=::NewEnumerator1
    Life Cycle Mode=::NewEnumerator1
    Loop Behavior=::NewEnumerator1
    Loop Duration Mode=::NewEnumerator1
    UseLoopDelay=false
   )
   &SpawnBurst_Instantaneous(Spawn Count=2, Use Spawn Probability=false, UseLoopCountLimit=false)
  ]
  particleSpawnScript [
   &FN_SetSpawnGroup(Number of Spawn Groups=2)
   /Niagara/Modules/Spawn/Initialization/V2/InitializeParticle(
    Lifetime=5
    Mesh Scale Mode=::NewEnumerator3
    Position Offset=&MergeHelper_VectorSwitch(0=<5.5 -16.8 -0.9>, 1=<-5.5 -16.8 -0.9>, 2=?, 3=?, Index=$Particles.SpawnGroup, Num=::NewEnumerator0)
    Position Offset Coordinate Space=::Local
    Sprite Size Mode=::NewEnumerator3
    Uniform Sprite Size=13
    UsePositionOffset=true
    Write Material Random=true
   )
   &SpriteFacingAndAlignment(Sprite Alignment=<0 0 1>, Sprite Facing=<0 1 0>, WriteAlignment=true, WriteFacing=true)
  ]
  particleUpdateScript [
   &ParticleState(Kill Particles When Lifetime Has Elapsed=false, Let Infinitely Lived Particles Die When Emitter Deactivates=true)
   &CameraOffset(Camera Offset Amount=5)
   &ScaleColor(
    Scale Alpha=&Multiply_Float(A=1, B=&FNTimeOfDayScalar(Night Scale=0.5, Morning Scale=0.7, Day Scale=0.8, Evening Scale=0.7))
    Scale RGB=&VectorFromFloat(Value=&Add_Float(
      A=&FloatFromCurve(
       FloatCurve=DataInterface() {propertyValues {Curve {keys [<at 0 0 RCIM_Cubic>,<at 0.9 1 RCIM_Cubic>,<at 1 0 RCIM_Cubic>]}}}
       CurveIndex=$System.AudioReactivity
       Scale Curve=&FN_FrontEndFloatModifier(Gameplay=1, Front End=1)
      )
      B=1
     ))
    ScaleA=true
    ScaleRGB=true
   ) {module=ScaleColor001}
   &ScaleSpriteSize(
    Uniform Curve Index=$System.AudioReactivity
    Uniform Curve Sprite Scale=DataInterface() {propertyValues {Curve {keys [{tangentMode=RCTM_User, time=0, value=1}, {tangentMode=RCTM_User, time=0.9, value=1.2}, <at 1 1>]}}}
   )
   &ScaleColor(Scale Alpha=$System.Timeline001.Time, ScaleA=true, ScaleRGB=false)
  ]
  rendererClasses [&NiagaraSpriteRendererProperties]
  renderers [{rendererClass=&NiagaraSpriteRendererProperties}]
 }
 PostLine002 {
  emitterUpdateScript [
   &EmitterState(Inactive Response=::NewEnumerator1, Life Cycle Mode=::NewEnumerator1, Loop Duration=0.5, UseLoopDelay=false)
   &SpawnBurst_Instantaneous(
    Spawn Count=2
    Spawn Probability=&MakeCustomFloatFromBool(Bool=&SetBoolByFloatComparison(A=$System.AudioReactivity, B=0.1), True Float=1, False Float=0)
    Use Spawn Probability=true
    UseLoopCountLimit=false
   )
   &SpawnBurst_Instantaneous(
    Spawn Count=2
    Spawn Probability=&MakeCustomFloatFromBool(Bool=&SetBoolByFloatComparison(A=$System.AudioReactivity, B=0.1), True Float=1, False Float=0)
    Spawn Time=0.2
    Use Spawn Probability=true
    UseLoopCountLimit=false
   ) {module=SpawnBurst_Instantaneous001}
  ]
  particleSpawnScript [
   &PartitionParticles(Particles Per Partition=1, Partition Mode:ENiagara_PartitionMode_v2=::NewEnumerator1)
   &InitializeParticle(
    Color=&MergeHelper_ColorSwitch(0=<rgba 0.6667 0 1 1>, 1=<rgba 0 0.6667 1 1>, 2=?, 3=?, Index=$Particles.PartitionParticles.Partition, Num=::NewEnumerator0)
    Lifetime=0.6
    Mesh Scale Mode=::NewEnumerator3
    Position Offset=&MergeHelper_VectorSwitch(0=<-7 -17 -10>, 1=<7 -17 -10>, 2=?, 3=?, Index=$Particles.PartitionParticles.Partition, Num=::NewEnumerator0)
    Sprite Size Mode=::NewEnumerator3
    Uniform Sprite Size=&MergeHelper_FloatSwitch(0=20, 1=20, 2=?, 3=?, Index=$Particles.PartitionParticles.Partition, Num=::NewEnumerator0)
    UsePositionOffset=true
    Write Material Random=true
   )
   /SetVariables(Particles.VisibilityTag=&MakeCustomIntFromBool(Bool=$User.bIsFrontEndPreview, True INT=1, False INT=0)) {isSetParameters=true}
  ]
  particleUpdateScript [
   &ParticleState()
   &SkeletalMeshLocation(
    Bone / Socket Index=&SelectIntFromArray(
     Direct Array Index=$Particles.PartitionParticles.Partition
     Direct Array Mode=::NewEnumerator1
     Int Selection Array=DataInterface() {propertyValues {IntData [8,9,10,11]}}
    )
    MaskSpawnGroup=false
    Mesh Sampling Type=::NewEnumerator5
    Sampled Position Offset=<0 0 0>
    Skeletal Mesh=$System.Skeletal Mesh
    Socket Sampling Mode=::NewEnumerator5
   ) {enabled=false, module=SkeletalMeshLocation002}
   &CameraOffset(Camera Offset Amount=20)
   /SetVariables(Particles.OffsetPosition:NiagaraPosition=&AddVectorToPosition(
     Position=$Particles.Position
     Vector=&MergeHelper_VectorSwitch(0=<10 0 5>, 1=<-10 0 5>, 2=?, 3=?, Index=$Particles.PartitionParticles.Partition, Num=::NewEnumerator0)
    )) {isSetParameters=true, module=SetVariables_#8}
   /SetVariables(
    Particles.WSOffsetPosition:NiagaraPosition=&TransformPosition(Position=$StackContext.OffsetPosition, Source Space=::Local, Destination Space=::World)
    Particles.WSPosition:NiagaraPosition=&TransformPosition(Position=$StackContext.Position, Source Space=::Local, Destination Space=::World)
   ) {isSetParameters=true}
   &DynamicMaterialParameters(
    Index 0 Param 1=&MakeFloatFromVector(Vector=Linked() {linkedVariable {name=StackContext.WSPosition, type=NiagaraPosition}}, Channel=::NewEnumerator0)
    Index 0 Param 2=&MakeFloatFromVector(Vector=Linked() {linkedVariable {name=StackContext.WSPosition, type=NiagaraPosition}}, Channel=::NewEnumerator1)
    Index 0 Param 3=&MakeFloatFromVector(Vector=Linked() {linkedVariable {name=StackContext.WSPosition, type=NiagaraPosition}}, Channel=::NewEnumerator2)
    Index 0 Param 4=&FloatFromCurve(FloatCurve=DataInterface() {propertyValues {Curve {keys [
         {arriveTangent=-0.1965, interpMode=RCIM_Cubic, leaveTangent=-0.1965, tangentMode=RCTM_User, time=0, value=1}
         {arriveTangent=-0.188, interpMode=RCIM_Cubic, leaveTangent=-0.188, tangentMode=RCTM_User, time=1, value=0.8}
        ]}}})
    Index 1 Param 1=&MakeFloatFromVector(Vector=Linked() {linkedVariable {name=StackContext.WSOffsetPosition, type=NiagaraPosition}}, Channel=::NewEnumerator0)
    Index 1 Param 2=&MakeFloatFromVector(Vector=Linked() {linkedVariable {name=StackContext.WSOffsetPosition, type=NiagaraPosition}}, Channel=::NewEnumerator1)
    Index 1 Param 3=&MakeFloatFromVector(Vector=Linked() {linkedVariable {name=StackContext.WSOffsetPosition, type=NiagaraPosition}}, Channel=::NewEnumerator2)
    Index 1 Param 4=&FloatFromCurve(
     FloatCurve=DataInterface() {propertyValues {Curve {keys [
         {arriveTangent=0.1333, interpMode=RCIM_Cubic, leaveTangent=0.1333, tangentMode=RCTM_User, time=0, value=0}
         {interpMode=RCIM_Cubic, leaveTangent=-0.5523, tangentMode=RCTM_Break, time=0.3, value=0.8}
         {arriveTangent=-0.6776, interpMode=RCIM_Cubic, leaveTangent=-0.6776, tangentMode=RCTM_User, time=1, value=0.2}
        ]}}}
     Scale Curve=2
    )
    Index 2 Param 1=&FloatFromCurve(FloatCurve=DataInterface() {propertyValues {Curve {keys [<at 0 0>,<at 1 1>]}}})
    Index 2 Param 4=1000
    Module.Index0Param0Write=true
    Module.Index0Param1Write=true
    Module.Index0Param2Write=true
    Module.Index0Param3Write=true
    Module.Index1Param0Write=true
    Module.Index1Param1Write=true
    Module.Index1Param2Write=true
    Module.Index1Param3Write=true
    Module.Index2Param0Write=true
    Module.Index2Param1Write=true
    Module.Index2Param2Write=true
    Module.Index2Param3Write=true
    Write Parameter Index 1=true
    Write Parameter Index 2=true
   )
   &ScaleColor(
    Scale Alpha=&FloatFromCurve(
     FloatCurve=DataInterface() {propertyValues {Curve {keys [
         {arriveTangent=19.3619, interpMode=RCIM_Cubic, leaveTangent=19.3619, tangentMode=RCTM_User, time=0, value=0}
         {arriveTangent=-2.6364, interpMode=RCIM_Cubic, leaveTangent=-2.6364, tangentMode=RCTM_User, time=0.15, value=0.6}
         {arriveTangent=-0.6911, interpMode=RCIM_Cubic, leaveTangent=-0.6911, tangentMode=RCTM_User, time=1, value=0}
        ]}}}
     Scale Curve=&Multiply_Float(A=50, B=$System.Timeline001.Time)
    )
    Scale RGB=&VectorFromFloat(Value=&FloatFromCurve(
      FloatCurve=DataInterface() {propertyValues {Curve {keys [
          {arriveTangent=0.168, interpMode=RCIM_Cubic, leaveTangent=0.168, tangentMode=RCTM_User, time=0.2, value=1.5}
          {arriveTangent=-0.1012, interpMode=RCIM_Cubic, leaveTangent=-0.1012, tangentMode=RCTM_User, time=0.4, value=1}
         ]}}}
      Scale Curve=0.5
     ))
    ScaleA=true
    ScaleRGB=true
   )
   &ScaleColor(
    Scale Alpha=&FloatFromCurve(
     FloatCurve=DataInterface() {propertyValues {Curve {keys [
         {interpMode=RCIM_Cubic, leaveTangent=0.431, tangentMode=RCTM_Break, time=0, value=0}
         {arriveTangent=3.1641, interpMode=RCIM_Cubic, leaveTangent=-6.0307, tangentMode=RCTM_Break, time=0.9, value=2}
         <at 1 1>
        ]}}}
     CurveIndex=$System.AudioReactivity
     Scale Curve=0.5
    )
    ScaleA=true
    ScaleRGB=false
   ) {module=ScaleColor003}
   &ScaleSpriteSize(
    Uniform Curve Index=$System.AudioReactivity
    Uniform Curve Sprite Scale=DataInterface() {propertyValues {Curve {keys [{tangentMode=RCTM_User, time=0, value=1}, {tangentMode=RCTM_User, time=0.9, value=2}, <at 1 1>]}}}
   ) {module=ScaleSpriteSize001}
  ]
  rendererClasses [&NiagaraSpriteRendererProperties]
  renderers [
   {rendererClass=&NiagaraSpriteRendererProperties}
   {rendererClass=&NiagaraSpriteRendererProperties, rendererIndex=1}
  ]
 }
}
summary {
 systemName=NS_Backpack_FareSpore
 emitters [
  {emitterName=Equalizer, rendererClasses [&NiagaraMeshRendererProperties]}
  {emitterName=Base, rendererClasses [&NiagaraMeshRendererProperties]}
  {emitterName=Arms_SoundWaves, rendererClasses [&NiagaraSpriteRendererProperties]}
  {emitterName=PostLine002, rendererClasses [&NiagaraSpriteRendererProperties]}
  {emitterName=Dots, rendererClasses [&NiagaraMeshRendererProperties]}
  {emitterName=Glow, rendererClasses [&NiagaraSpriteRendererProperties]}
  {bEnabled=false, emitterName=FN_Debug, rendererClasses [&NiagaraSpriteRendererProperties]}
 ]
 userVariables [
  User.bIsADS {type=NiagaraBool}
  User.bisFootHidden {type=NiagaraBool}
  User.bIsFrontEnd {type=NiagaraBool}
  User.bIsFrontEndPreview {type=NiagaraBool}
  User.bisHeadHidden {type=NiagaraBool}
  User.DebugReactivity {type=NiagaraBool}
  User.EnableFrontendFakeAudioLoop {type=NiagaraFloat, defaultValue=0, description=""}
 ]
}
