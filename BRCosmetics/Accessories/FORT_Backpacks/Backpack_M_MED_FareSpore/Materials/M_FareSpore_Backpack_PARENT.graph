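# Parent material for the FareSpore backpack, layering switchable audio-reactive effects between the 2023 character parent PreFX and PostFX functions.
# The frontend-loop audio amplitude drives a noise-warped mask threshold and thin-film intensity.
# Scrolling pink/blue gradients, sine-animated color wheels, a mouth mask and a play/pause icon build masked emissive, with masked base-color darkening.
# Helper coordinates include camera-facing virtual-plane UVs with FOV and distance scaling, radial normal-view coordinates and depth-scaled halftones.
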Material /BRCosmetics/Accessories/FORT_Backpacks/Backpack_M_MED_FareSpore/Materials/M_FareSpore_Backpack_PARENT
graph {
 properties {
  bUseMaterialAttributes=true
  bUsedWithMeshDeformer=true
  bUsedWithMorphTargets=true
  bUsedWithSkeletalMesh=true
  pixelDepthOffsetMode=PDOM_Legacy
  shadingModel=MSM_FromMaterialExpression
  usedShadingModels=Default Lit | Subsurface | Clear Coat | Subsurface Profile | Cloth
 }
 nodes [
  ComponentMask(None=NamedRerouteUsage()) {b=true, r=true}
  Constant() {r=1}
  n1 = MaterialFunctionCall() {materialFunction=&MF_CharacterParent_2023_PreFX}
  MaterialFunctionCall(
   Material=StaticSwitchParameter(True=NamedRerouteUsage(), False=NamedRerouteDeclaration(None=@n1.Result) {name=OriginalMA, nodeColor=<rgba 0 0.2969 1 1>}) {defaultValue=true, group=FareSpore, parameterName=IsAudioReactive, sortPriority=-10000}
   M=@n1.M
   Pre Skinned Local Position=NamedRerouteDeclaration(None=@n1.Pre skinned Local Position) {name=PreskinnedLocalPosition, nodeColor=<rgba 1 0.0234 0 1>}
   Distance Blend=@n1.M
   Fresnel=NamedRerouteDeclaration(None=@n1.Fresnel) {name=InFresnel, nodeColor=<rgba 0 1 0.7422 1>}
  ) {materialFunction=&MF_CharacterParent_2023_PostFX}
  Multiply(A=TextureCoordinate()) {constB=30}
  n2 = Constant() {r=0.2}
  Multiply(A=@n2, B=Constant2Vector() {g=1.1, r=1})
  n3 = Constant() {r=0.36}
  Multiply(A=@n3, B=Constant2Vector() {g=1.1, r=1})
  n4 = MaterialFunctionCall!(InAmplitude, InSpectrum, CustomFakeSpectrum, CustomFakeAmplitude) {materialFunction=/BRCosmetics/Global/MaterialFunctions/AudioReactivity/MF_AudioReactivity_FrontendLoop}
  n5 = Power(Base=@n4.Amplitude) {constExponent=3}
  n6 = Power(Base=NamedRerouteUsage(), Exp=ScalarParameter() {defaultValue=3, group=06_Thin Film, parameterName=CustomThinFilmExponent})
  n7 = ScalarParameter() {defaultValue=0.3, group=06_Thin Film, parameterName=CustomThinFilmIntensityExponent, sortPriority=3}
  n8 = Multiply(
   A=LinearInterpolate(
    A=ScalarParameter() {defaultValue=1, group=06_Thin Film, parameterName=CustomThinFilmIntensity, sortPriority=2}
    B=ScalarParameter() {defaultValue=2, group=06_Thin Film, parameterName=CustomThinFilmIntensity_Loud, sortPriority=2}
    Alpha=Multiply(A=@n5, B=@n6)
   )
   B=SmoothStep(Value=Power(Base=@n6, Exp=@n7))
  )
  n9 = NamedRerouteUsage()
  n10 = ScalarParameter() {defaultValue=0.5, group=06_Thin Film, parameterName=CustomThinFilmWarp, sortPriority=4}
  n11 = Add(
   A=Multiply(A=ScalarParameter() {defaultValue=-0.1, group=06_Thin Film, parameterName=CustomThinFilmPanSpeed}, B=Time())
   B=Multiply(A=ScalarParameter() {defaultValue=0.5, group=06_Thin Film, parameterName=CustomThinFilmScale, sortPriority=5}, B=Power(Base=@n9, Exp=@n10))
  )
  n12 = ScalarParameter() {defaultValue=0.3, group=06_Thin Film, parameterName=CustomThinFilmOffset}
  n13 = ComponentMask(None=NamedRerouteUsage()) {g=true}
  n14 = Multiply(A=Time(), B=ScalarParameter() {defaultValue=0.1, parameterName=CustomThinFilmCircularMaskRotateSpeed})
  n15 = TextureCoordinate()
  n16 = TextureSample!(
   UVs=AppendVector(A=Add(A=@n11, B=@n12), B=Subtract(A=@n13, B=@n14))
   Tex
   "DDX(UVs)"=DDX(Value=@n15)
   "DDY(UVs)"=DDY(Value=@n15)
   Apply View MipBias
  ) {
   mipValueMode=TMVM_Derivative
   samplerSource=SSM_Wrap_WorldGroupSettings
   texture=/BRCosmetics/Characters/Player/Male/Medium/Bodies/M_MED_FareSpore/Textures/T_FareSpore_FaceAcc_GradientMap
  }
  n17 = Multiply(A=Time(), B=ScalarParameter() {defaultValue=0.1, parameterName=CustomThinFilmCircularPanSpeed})
  n18 = CurveAtlasRowParameter(CurveTime=Frac(None=Add(B=@n17))) {
   atlas=/Game/Weapons/FORT_Melee/Pickaxe_RelayStick/Textures/CA_RelayStick
   curve=/Game/Weapons/FORT_Melee/Pickaxe_RelayStick/Textures/C_RelayStick_CircularFresnel
   defaultValue=1
   parameterName=CustomThinFilmRadialGradient
  }
  n19 = Add(
   A=@n16.RGB
   B=MaterialFunctionCall(R=@n18.R, G=Constant(), B=@n18.R) {materialFunction=&AppendMany}.RGB
  )
  n20 = NamedRerouteDeclaration(None=Desaturation(None=Multiply(A=@n19), Fraction=Constant() {r=0.3})) {name=SpectralThinFilm, nodeColor=<rgba 0.8516 1 0 1>}
  NamedRerouteDeclaration(None=Multiply(A=Multiply(A=@n8, B=@n20)) {constB=0.3}) {name=ColorfulThinFilm, nodeColor=<rgba 0 1 0.9531 1>}
  NamedRerouteDeclaration(None=Constant3Vector() {constant=<rgba 0.0073 0 1 1>}) {name=ColorWheelColor_Edge, nodeColor=<rgba 1 0.0938 0.6978 1>}
  NamedRerouteDeclaration(None=Power(Base=@n4.Amplitude)) {name=AudioAmplitude, nodeColor=<rgba 1 0 0.0469 1>}
  n21 = TextureCoordinate()
  n22 = Constant2Vector() {g=0.4, r=-0.275}
  n23 = LinearInterpolate(A=@n22, B=Add(A=@n22, B=Constant2Vector() {r=-0.038}), Alpha=Ceil(None=NamedRerouteUsage()))
  n24 = TextureSample!(
   UVs=MaterialFunctionCall(UVs=Add(A=@n21, B=@n23), Texture Scale=Constant() {r=-0.08}) {materialFunction=&ScaleUVsByCenter}.UVs
   Tex
   Apply View MipBias
  ) {
   samplerSource=SSM_Clamp_WorldGroupSettings
   samplerType=SAMPLERTYPE_LinearColor
   texture=/BRCosmetics/Accessories/FORT_Backpacks/Backpack_M_MED_FareSpore/Textures/T_FareSpore_Backpack_AudioReactivityFX
  }
  n25 = MaterialFunctionCall!(
   Texture Coordinate=TextureCoordinate()
   X Pos Right=Constant() {r=0.82}
   X Pos Left=Constant() {r=0.775}
   Y Pos Up=Constant() {r=0.07}
   Y Pos Down=Constant() {r=0.13}
   SquareSoftness
   RotateSquare
   RotationAmount
  ) {materialFunction=/Game/Accessories/FORT_Backpacks/Backpack_F_MED_GroovyReader_Play/Materials/MaterialFunctions/MF_SquareMaskGenerator}
  NamedRerouteDeclaration(None=Multiply(A=Multiply(A=@n24.B, B=@n25.SquareMask), B=Constant3Vector() {constant=<rgba 0.4531 1 0.8361 1>})) {name=Play/Pause, nodeColor=<rgba 0 0.2266 1 1>}
  n26 = ComponentMask(None=Divide(A=NamedRerouteUsage()) {constB=100}) {b=true, r=true}
  n27 = StaticSwitchParameter(True=MaterialFunctionCall(XY=@n26) {materialFunction=&Swizzle}.YX, False=@n26) {group=FareSpore, parameterName=Is Guitar?}
  n28 = Constant2Vector() {g=0.2, r=0.2}
  n29 = TextureSample(UVs=Panner(Coordinate=@n27, Speed=@n28) {speedX=0.5}) {
   samplerSource=SSM_Wrap_WorldGroupSettings
   texture=/BRCosmetics/Accessories/FORT_Backpacks/Backpack_M_MED_FareSpore/Textures/T_FareSpore_Backpack_AudioReactivityGradient
  }
  n30 = ScalarParameter() {defaultValue=3, parameterName=Pink/Blue_ScrollingSpeed}
  n31 = Time()
  n32 = TextureSample(UVs=Panner(Coordinate=@n27, Time=Add(A=@n30, B=@n31), Speed=Multiply(A=@n28) {constB=-1}) {speedX=-0.5}) {
   samplerSource=SSM_Wrap_WorldGroupSettings
   texture=&T_FareSpore_Backpack_AudioReactivityGradient
  }
  n33 = ComponentMask(None=NamedRerouteUsage()) {r=true}
  n34 = Saturate(None=Add(A=Multiply(A=@n33) {constB=0.1}) {constB=0.4})
  NamedRerouteDeclaration(None=Desaturation(None=LinearInterpolate(A=@n29.RGB, B=@n32.RGB, Alpha=@n34), Fraction=Constant() {r=-0.1})) {name="Pink&&BlueColors", nodeColor=<rgba 0.4531 0 1 1>}
  n35 = NamedRerouteUsage()
  n36 = TransformPosition(None=NamedRerouteUsage()) {transformSourceType=TRANSFORMPOSSOURCE_World, transformType=TRANSFORMPOSSOURCE_View}
  n37 = Multiply(A=Divide(B=ComponentMask(None=@n36) {b=true})) {constB=0.8}
  n38 = TextureSample(UVs=Panner(Coordinate=MaterialFunctionCall(UVs=@n35, Texture Scale=@n37) {materialFunction=&ScaleUVsByCenter}.UVs) {speedY=1.7}) {
   samplerSource=SSM_Wrap_WorldGroupSettings
   samplerType=SAMPLERTYPE_LinearGrayscale
   texture=/Game/Characters/Player/Female/Medium/Bodies/F_MED_TNTina/Materials/HalftoneOffset
  }
  NamedRerouteDeclaration(None=@n38.R) {name=Halftones, nodeColor=<rgba 1 0.2344 0 1>}
  n39 = GetMaterialAttributes(None=NamedRerouteUsage()) {attributeGetTypes [#1, #2]}
  n40 = OneMinus(None=Power(Base=OneMinus(None=@n4.Amplitude)) {constExponent=4})
  n41 = TextureCoordinate()
  n42 = TextureSample(UVs=Panner() {speedX=-0.05, speedY=-0.015}) {
   samplerSource=SSM_Wrap_WorldGroupSettings
   samplerType=SAMPLERTYPE_LinearColor
   texture=&T_FareSpore_Backpack_AudioReactivityFX
  }
  n43 = Multiply(A=@n42.G) {constB=2}
  n44 = Multiply(A=Multiply(A=Add(A=@n43) {constB=-1}, B=Constant2Vector() {g=0.1, r=1})) {constB=0.015}
  n45 = TextureSample(UVs=Add(A=Add(A=@n41, B=@n44), B=Constant2Vector() {r=0.0035})) {
   samplerSource=SSM_Wrap_WorldGroupSettings
   samplerType=SAMPLERTYPE_LinearColor
   texture=&T_FareSpore_Backpack_AudioReactivityFX
  }
  n46 = SmoothStep(Min=LinearInterpolate(Alpha=@n40) {constA=0.2, constB=0.35}, Max=LinearInterpolate(Alpha=@n40) {constA=0.43, constB=0.5}, Value=@n45.R) {constMax=0.5, constMin=0.35}
  n47 = Saturate(None=Add(A=OneMinus(None=@n46)) {constB=0.03})
  n48 = Saturate(None=Multiply(A=NamedRerouteUsage()) {constB=2})
  n49 = Add(A=Multiply(A=NamedRerouteUsage()) {constB=1.5}, B=NamedRerouteUsage())
  n50 = Multiply(A=@n49, B=Add(A=NamedRerouteUsage()))
  n51 = StaticSwitchParameter(True=Constant(), False=NamedRerouteUsage()) {group=FareSpore, parameterName=Disable Mouth Mask}
  n52 = NamedRerouteUsage()
  n53 = Desaturation(None=Multiply(A=LinearInterpolate(A=@n50, B=@n49, Alpha=@n51), B=ComponentMask(None=@n52) {b=true}), Fraction=Constant() {r=-1})
  n54 = StaticSwitchParameter(True=Add(A=@n53, B=NamedRerouteUsage()), False=@n53) {defaultValue=true, group=FareSpore, parameterName=Enable Spinning Wheels}
  n55 = NamedRerouteUsage()
  n56 = Add(A=@n39.Emissive Color, B=Multiply(A=Multiply(A=@n47, B=@n48), B=Multiply(A=@n54, B=@n55)))
  n57 = Add(A=NamedRerouteUsage(), B=@n56)
  n58 = NamedRerouteDeclaration(None=@n39.Base Color) {name=OriginalBaseColor, nodeColor=<rgba 1 0.9141 0 1>}
  n59 = OneMinus(None=ComponentMask(None=NamedRerouteUsage()) {b=true})
  n60 = Multiply(A=@n58, B=Saturate(None=Add(A=@n59) {constB=0.05}))
  n61 = Ceil(None=NamedRerouteUsage())
  NamedRerouteDeclaration(None=SetMaterialAttributes!(
    MaterialAttributes=@n39.MaterialAttributes
    Emissive Color=StaticSwitchParameter(True=@n57, False=@n56) {defaultValue=true, group=FareSpore, parameterName=Enable Play/Pause}
    Base Color=LinearInterpolate(A=@n58, B=@n60, Alpha=@n61)
   ) {attributeSetTypes [#1, #2]}) {name=OutputMA, nodeColor=<rgba 0.7812 1 0 1>}
  n62 = MaterialFunctionCall() {materialFunction=&CameraDirectionVector}
  n63 = Multiply(A=Constant3Vector() {constant=<rgba 1 1 0 0>}, B=@n62.Result)
  n64 = Add(A=Multiply(B=Normalize(VectorInput=@n63)) {constA=0.25}, B=@n62.Result)
  n65 = VectorParameter() {group=Superlevel, parameterName=VirtualPlanePosition, sortPriority=1000, defaultValue=<rgba 0 -15 0 0>}
  n66 = NamedRerouteDeclaration(None=TransformPosition(None=@n65.RGB) {transformType=TRANSFORMPOSSOURCE_World}) {name=BackpackPosition, nodeColor=<rgba 0 1 0.2734 1>}
  n67 = Constant() {r=45}
  n68 = Constant() {r=90}
  n69 = MaterialFunctionCall() {materialFunction=/Engine/Functions/Engine_MaterialFunctions02/Utility/FOV}
  n70 = Multiply(
   A=LinearInterpolate(A=Constant() {r=64}, B=Constant() {r=128}, Alpha=InverseLinearInterpolate(A=@n67, B=@n68, Value=@n69.FOV angle in degrees))
   B=Distance(A=@n66, B=CameraPositionWS())
  )
  NamedRerouteDeclaration(None=MaterialFunctionCall(Plane Normal Axis=Normalize(VectorInput=@n64), Plane Center=Add(A=@n62.Result, B=@n66), UV Scale=Multiply(A=@n70) {constB=0.0146}) {materialFunction=&VirtualPlaneCoordinates}.UVs) {name=BackpackAllignedScreenPosition, nodeColor=<rgba 0.125 1 0 1>}
  NamedRerouteDeclaration(None=@n1.Tech Art Mask) {name=TechArtMask, nodeColor=<rgba 0.3359 0 1 1>}
  n71 = TextureCoordinate()
  n72 = Constant() {r=0.87}
  n73 = Constant() {r=0.67}
  n74 = Constant() {r=0.13}
  n75 = Constant() {r=0.5}
  n76 = NamedRerouteUsage()
  NamedRerouteDeclaration(None=Multiply(
    A=MaterialFunctionCall(Texture Coordinate=@n71, X Pos Right=@n72, X Pos Left=@n73, Y Pos Up=@n74, Y Pos Down=@n75) {materialFunction=&MF_SquareMaskGenerator}.SquareMask
    B=ComponentMask(None=@n76) {b=true}
   )) {name=MouthMask, nodeColor=<rgba 1 0 0.4922 1>}
  n77 = TextureCoordinate()
  n78 = Constant2Vector() {g=0.285, r=0.32}
  n79 = TextureSample(UVs=MaterialFunctionCall(UVs=Add(A=@n77, B=@n78), Texture Scale=@n3) {materialFunction=&ScaleUVsByCenter}.UVs) {
   samplerSource=SSM_Clamp_WorldGroupSettings
   texture=/BRCosmetics/Characters/Player/Male/Medium/Bodies/M_MED_FareSpore/Skins/Spill/Textures/T_FareSpore_Spill_AudioWheel
  }
  n80 = Add(A=Subtract(A=Time(), B=NamedRerouteUsage()), B=@n79.R)
  n81 = Add(A=Multiply(A=Sine(None=@n80)) {constB=0.5}) {constB=0.5}
  n82 = LinearInterpolate(A=NamedRerouteUsage(), B=NamedRerouteUsage(), Alpha=Saturate(None=Add(A=@n81) {constB=-0.1}))
  n83 = Saturate(None=Add(A=@n81) {constB=-0.2})
  n84 = LinearInterpolate(Alpha=@n81) {constA=0.45, constB=0.2}
  n85 = TextureCoordinate()
  n86 = Constant2Vector() {g=0.29, r=0.315}
  n87 = TextureSample(UVs=MaterialFunctionCall(UVs=Add(A=@n85, B=@n86), Texture Scale=@n2) {materialFunction=&ScaleUVsByCenter}.UVs) {
   samplerSource=SSM_Clamp_WorldGroupSettings
   texture=&T_FareSpore_Spill_AudioWheel
  }
  n88 = Subtract(A=Subtract(A=Time(), B=@n87.R), B=NamedRerouteUsage())
  n89 = Add(A=Multiply(A=Sine(None=@n88)) {constB=0.5}) {constB=0.5}
  n90 = LinearInterpolate(A=NamedRerouteUsage(), B=NamedRerouteUsage(), Alpha=Saturate(None=Add(A=@n89) {constB=-0.3}))
  n91 = Saturate(None=Add(A=@n89) {constB=-0.6})
  n92 = LinearInterpolate(Alpha=@n89) {constA=0.45, constB=0.2}
  n93 = Add(
   A=Multiply(A=Multiply(A=@n82, B=@n83), B=SmoothStep(Min=@n84, Value=@n79.G) {constMax=0.5, constMin=0.6})
   B=Multiply(A=Multiply(A=@n90, B=@n91), B=SmoothStep(Min=@n92, Value=@n87.G) {constMax=0.5, constMin=0.6})
  )
  NamedRerouteDeclaration(None=Desaturation(None=Multiply(A=@n93) {constB=50}, Fraction=Constant() {r=-1})) {name=SpinningWheels, nodeColor=<rgba 1 0.5156 0 1>}
  NamedRerouteDeclaration(None=Constant3Vector() {constant=<rgba 1 0.007 0.5072 1>}) {name=ColorWheelColor_Center, nodeColor=<rgba 0.1474 0.0469 1 1>}
  n94 = DotProduct(A=PixelNormalWS(), B=CameraVectorWS())
  n95 = Subtract(A=PixelNormalWS(), B=CameraVectorWS())
  n96 = Normalize(VectorInput=Multiply(
    A=Transform(None=@n95) {transformSourceType=TRANSFORMSOURCE_World, transformType=TRANSFORM_Camera}
    B=Constant3Vector() {constant=<rgba 1 1 0 0>}
   ))
  n97 = Constant3Vector() {constant=<rgba 0 1 0 0>}
  n98 = ComponentMask(None=CrossProduct(A=@n96, B=@n97)) {b=true}
  n99 = Abs(None=@n98)
  n100 = MaterialFunctionCall(In=@n99, Power=Constant() {r=0.75}) {
   desc="Kind of optional, but smoothes out\r\nsome nonlinear distortion"
   materialFunction=&SCurve
  }
  n101 = Multiply(A=QualitySwitch(Default=@n100.Result, Low=@n99, Medium=@n99), B=Sign(None=@n98))
  n102 = ConstantBiasScale(None=@n101) {bias=-3, scale=-0.25}
  n103 = ConstantBiasScale(None=@n101) {scale=0.25}
  n104 = Saturate(None=Sign(None=DotProduct(A=@n96, B=@n97)))
  NamedRerouteDeclaration(None=AppendVector(A=Power(Base=@n94) {desc=Radial coordinate distortion}, B=LinearInterpolate(A=@n102, B=@n103, Alpha=@n104) {constB=0}) {desc=Output UVs}) {name=RadialCoords, nodeColor=<rgba 0.7578 0 1 1>}
 ]
}
